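# Niagara effect for the Kittybasic vehicle boost cosmetic, trailing meshes, sprites and multicolored confetti behind the car.
# User.bIsFrontendPreview switches spawning between two mesh emitters, one for gameplay and one for frontend preview, both killing particles on collision.
# A GPU confetti emitter layers two spawn groups of spinning, shrinking sprites.
# Emitters offset from SingleBoostPos plus DTCarOffset and inherit source movement; socket sampling and painted-variant modules are disabled or absent.
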NiagaraSystem /VehicleCosmetics/Boosts/Boost_Kittybasic/FX/NS_Boost_Kittybasic
dependencies {
 usedDataInterfaces [
  &NiagaraDataInterfaceArrayColor
  &NiagaraDataInterfaceCurve
  &NiagaraDataInterfaceVector2DCurve
  &NiagaraDataInterfaceVectorCurve
  &NiagaraDataInterfaceArrayFloat
  &NiagaraDataInterfaceSkeletalMesh
 ]
 usedDynamicInputs []
 usedModules [
  &self:SystemSpawnScript.SystemScriptSource.SystemScriptGraph.NiagaraNodeAssignment_0.SetVariables_#1
  &SystemState
 ]
}
emitters {
 NE_VC_Boost_Mesh_Box {
  emitterUpdateScript [
   &EmitterState(Inactive Response=::NewEnumerator1, Life Cycle Mode=::NewEnumerator1, Loop Duration=0.1, UseLoopDelay=false)
   &VC_FilteredSocketCount()
   &SpawnRate(
    SpawnRate=&MakeCustomFloatFromBool(Bool=$User.bIsFrontendPreview, True Float=0, False Float=30)
    Use Spawn Probability=false
   )
  ]
  particleSpawnScript [
   &VC_PaintedVariantsOverride(Crimson=<rgba 1 0.03 0.0462 1>, DefaultColor=<rgba 0.6333 0.6 1 1>) {enabled=false}
   &InitializeParticle(
    Lifetime=&MakeCustomFloatFromBool(Bool=$User.bIsFrontendPreview, True Float=0.7, False Float=1)
    Mesh Scale Mode=::NewEnumerator3
    Mesh Uniform Scale=2
    Position Offset=&AddVectorToPosition(
     Position=&AddVectorToPosition(Position=<-25 0 5>, Vector=$User.SingleBoostPos)
     Vector=&Multiply_VectorByFloat(Vector=$User.DTCarOffset, Float=1)
    )
    Position Offset Coordinate Space=::Local
    Sprite Rotation Mode=::NewEnumerator2
    Sprite Size Min=<2 5>
    Sprite Size Mode=::NewEnumerator2
    UsePositionOffset=true
    Write Material Random=true
   )
   &PartitionParticles(Number Of Partitions=$Emitter.PortCount) {enabled=false}
   &SkeletalMeshLocation(
    Bone / Socket Index=$Particles.PartitionParticles.Partition
    Bone / Socket Sampling Mode=::NewEnumerator5
    MaskSpawnGroup=false
    Mesh Sampling Type=::NewEnumerator6
    Sampled Position Offset=/Niagara/DynamicInputs/TypeConversions/MakeCustomPositionFromBool(
     Bool:NiagaraBool=$User.bIsFrontendPreview
     True Position:NiagaraPosition=&AddVectorToPosition(Position=<-25 0 0>, Vector=$User.DTCarOffset)
     False Position:NiagaraPosition=&AddVectorToPosition(Position=<-50 0 0>, Vector=$User.DTCarOffset)
    )
    Skeletal Mesh=$System.SkeletalMesh
   ) {enabled=false}
   &AddVelocity(
    Velocity=&VectorFromCurve(
     VectorCurve=DataInterface() {propertyValues {
       XCurve {keys [<at 0 1>,<at 1 0>]}
       YCurve {keys [<at 0 1>,<at 1 1>]}
       ZCurve {keys [<at 0 1 RCIM_Cubic>, {arriveTangentWeight=0.4904, interpMode=RCIM_Cubic, tangentMode=RCTM_User, time=1, value=1}]}
      }}
     Scale Curve=<1800 1 150>
    )
    Velocity Speed Scale=-1
   )
   &ShapeLocation(MaskSpawnGroup=false, Override Seed=false, Sphere Radius=35, UseOwnerScale=true) {enabled=false}
   &InitialMeshOrientation(Rotation=<1 1 1>, Use Rotation Vector=false)
  ]
  particleUpdateScript [
   &ParticleState()
   &UpdateMeshOrientation(Rotation Rate=2)
   &InheritSourceMovement(
    Applied Position Scale=&Multiply_VectorByFloat(
     Vector=<1 1 1>
     Float=&FloatFromCurve(FloatCurve=DataInterface() {propertyValues {Curve {keys [
          {arriveTangent=-0.5211, interpMode=RCIM_Cubic, leaveTangent=-0.5211, tangentMode=RCTM_User, time=0, value=1}
          {arriveTangentWeight=0.3344, interpMode=RCIM_Cubic, tangentMode=RCTM_User, time=1, value=0}
         ]}}})
    )
    ApplyForce=false
    ApplyPosition=true
    ApplyPositionFalloff=false
   )
   &Collision(
    Age Colliding Particles=false
    KillOnCollision=true
    Module.Control Rest State Via Surface Normal=false
    Module.EnableMaxCollisionCount=false
    Module.Randomize Collision Normal Vector=false
    Restitution=0
   )
   &Drag(
    Drag=&FloatFromCurve(FloatCurve=DataInterface() {propertyValues {Curve {keys [<at 0 0 RCIM_Cubic>,<at 1 1 RCIM_Cubic>]}}}, Scale Curve=2)
    Use Linear Drag=true
    Use Rotational Drag=true
   )
   &SolveForcesAndVelocity(Clamp Velocity=false, Limit Acceleration=false)
   &DynamicMaterialParameters(
    Index 0 Param 1=1
    Index 0 Param 2=1
    Index 0 Param 3=1
    Index 0 Param 4=1
    Module.Index0Param0Write=true
    Module.Index0Param1Write=true
    Module.Index0Param2Write=true
    Module.Index0Param3Write=true
   ) {enabled=false}
   &ScaleColor(
    Scale RGB=&Multiply_VectorByFloat(
     Vector=<1 1 1>
     Float=&MakeCustomFloatFromBool(Bool=$User.bIsFrontendPreview, True Float=1, False Float=1)
    )
    ScaleA=true
    ScaleRGB=true
   ) {enabled=false}
  ]
  rendererClasses [&NiagaraMeshRendererProperties, &NiagaraSpriteRendererProperties]
  renderers [
   {rendererClass=&NiagaraMeshRendererProperties, rendererIndex=0}
   {rendererClass=&NiagaraSpriteRendererProperties, rendererIndex=1}
  ]
 }
 NE_VC_Boost_Mesh_Box001 {
  emitterUpdateScript [
   &EmitterState(Inactive Response=::NewEnumerator1, Life Cycle Mode=::NewEnumerator1, Loop Duration=0.1, UseLoopDelay=false)
   &VC_FilteredSocketCount() {enabled=false}
   &SpawnRate(
    SpawnRate=&MakeCustomFloatFromBool(Bool=$User.bIsFrontendPreview, True Float=17, False Float=0)
    Use Spawn Probability=false
   )
  ]
  particleSpawnScript [
   &VC_PaintedVariantsOverride(Crimson=<rgba 1 0.03 0.0462 1>, DefaultColor=<rgba 0.6333 0.6 1 1>) {enabled=false}
   &InitializeParticle(
    Color=<rgba 0.98 0.6 1 1>
    Lifetime=&MakeCustomFloatFromBool(Bool=$User.bIsFrontendPreview, True Float=&RandomRangeFloat(Minimum=2, Maximum=1), False Float=0)
    Mesh Scale Mode=::NewEnumerator3
    Mesh Uniform Scale=2
    Position Offset=&AddVectorToPosition(
     Position=&AddVectorToPosition(Position=<-25 0 5>, Vector=$User.SingleBoostPos)
     Vector=&Multiply_VectorByFloat(Vector=$User.DTCarOffset, Float=1)
    )
    Position Offset Coordinate Space=::Local
    Sprite Rotation Mode=::NewEnumerator2
    Sprite Size Min=<2 5>
    Sprite Size Mode=::NewEnumerator2
    UsePositionOffset=true
    Write Material Random=true
   )
   &PartitionParticles(Number Of Partitions=$Emitter.PortCount) {enabled=false}
   &SkeletalMeshLocation(
    Bone / Socket Sampling Mode=::NewEnumerator5
    MaskSpawnGroup=false
    Mesh Sampling Type=::NewEnumerator6
    Sampled Position Offset=&MakeCustomPositionFromBool(
     Bool:NiagaraBool=$User.bIsFrontendPreview
     True Position:NiagaraPosition=&AddVectorToPosition(Position=<-25 0 0>, Vector=$User.DTCarOffset)
     False Position:NiagaraPosition=&AddVectorToPosition(Position=<-50 0 0>, Vector=$User.DTCarOffset)
    )
    Skeletal Mesh=$System.SkeletalMesh
   ) {enabled=false}
   &AddVelocity(
    Cone Angle=6
    Cone Axis=&RandomRangeVector(Minimum=<-1 -0.05 0.1>, Maximum=<-1 -0.05 0.05>)
    Module.UseConeFalloff=false
    Module.UseVelDistribution=false
    Override Seed=false
    Velocity Mode=::NewEnumerator2
    Velocity Speed=&SelectFloatFromArray(
     Array Sampling Mode=::NewEnumerator0
     Direct Array Index=?
     Float Selection Array=DataInterface() {propertyValues {FloatData [900,1200]}}
     Random Seed=0
    )
   )
   &ShapeLocation(MaskSpawnGroup=false, Override Seed=false, Sphere Radius=35, UseOwnerScale=true) {enabled=false}
   &InitialMeshOrientation(Rotation=<1 1 1>, Use Rotation Vector=false)
  ]
  particleUpdateScript [
   &ParticleState()
   &ScaleMeshSize(Scale Factor=&VectorFromFloat(Value=&FloatFromCurve(FloatCurve=DataInterface() {propertyValues {Curve {keys [<at 0 1 RCIM_Cubic>,<at 1 1.5 RCIM_Cubic>]}}}))) {enabled=false}
   &UpdateMeshOrientation(Rotation Rate=0.75)
   &ScaleVelocity(
    Coordinate Space:ENiagaraCoordinateSpace=::Simulation
    Velocity Scale=&VectorFromCurve(
     VectorCurve=DataInterface() {propertyValues {
       XCurve {keys [<at 0 1.5>,<at 1 1>]}
       YCurve {keys [<at 0 1>,<at 1 1>]}
       ZCurve {keys [
         {arriveTangent=-12.8217, interpMode=RCIM_Cubic, leaveTangent=-12.8217, tangentMode=RCTM_User, time=0, value=0}
         {arriveTangent=-17.5003, interpMode=RCIM_Cubic, leaveTangent=-17.5003, tangentMode=RCTM_User, time=1, value=8}
        ]}
      }}
     Scale Curve=&RandomRangeVector(Minimum=<1 1 -2>, Maximum=<1.1 1 -2.5>)
    )
   )
   &InheritSourceMovement(
    Applied Position Scale=&Multiply_VectorByFloat(
     Vector=<1 1 1>
     Float=&FloatFromCurve(FloatCurve=DataInterface() {propertyValues {Curve {keys [
          {arriveTangent=-0.5211, interpMode=RCIM_Cubic, leaveTangent=-0.5211, tangentMode=RCTM_User, time=0, value=1}
          {arriveTangentWeight=0.3344, interpMode=RCIM_Cubic, tangentMode=RCTM_User, time=1, value=0}
         ]}}})
    )
    ApplyForce=false
    ApplyPosition=true
    ApplyPositionFalloff=false
   )
   &Collision(
    Age Colliding Particles=false
    KillOnCollision=true
    Module.Control Rest State Via Surface Normal=false
    Module.EnableMaxCollisionCount=false
    Module.Randomize Collision Normal Vector=false
    Restitution=0
   )
   &Drag(
    Drag=0.5
    Rotational Drag:NiagaraFloat=&FloatFromCurve(FloatCurve=DataInterface() {propertyValues {Curve {keys [<at 0 1 RCIM_Cubic>,<at 1 0 RCIM_Cubic>]}}}, Scale Curve=5)
    Use Linear Drag=true
    Use Rotational Drag=true
   )
   &SolveForcesAndVelocity(Clamp Velocity=false, Limit Acceleration=false)
   &DynamicMaterialParameters(
    Index 0 Param 1=1
    Index 0 Param 2=1
    Index 0 Param 3=1
    Index 0 Param 4=1
    Module.Index0Param0Write=true
    Module.Index0Param1Write=true
    Module.Index0Param2Write=true
    Module.Index0Param3Write=true
   ) {enabled=false}
   &ScaleColor(
    Scale RGB=&Multiply_VectorByFloat(
     Vector=<1 1 1>
     Float=&MakeCustomFloatFromBool(Bool=$User.bIsFrontendPreview, True Float=1, False Float=1)
    )
    ScaleA=true
    ScaleRGB=true
   )
  ]
  rendererClasses [&NiagaraMeshRendererProperties, &NiagaraSpriteRendererProperties]
  renderers [
   {rendererClass=&NiagaraMeshRendererProperties, rendererIndex=0}
   {rendererClass=&NiagaraSpriteRendererProperties, rendererIndex=1}
  ]
 }
 Sprite-Confetti {
  simTarget=GPUComputeSim
  emitterUpdateScript [
   &EmitterState(Inactive Response=::NewEnumerator1, Life Cycle Mode=::NewEnumerator1, Loop Duration=0.05, UseLoopDelay=false)
   &VC_FilteredSocketCount()
   &SpawnRate(SpawnRate=1000, Use Spawn Probability=false)
   &SpawnRate(SpawnGroup=1, SpawnRate=700, Use Spawn Probability=false) {module=SpawnRate001}
  ]
  particleSpawnScript [
   /SetVariables(Particles.VisTag:NiagaraInt32=&SpawnGroup()) {isSetParameters=true}
   &VC_PaintedVariantsOverride(DefaultColor=<rgba 1 0 0 1>) {enabled=false}
   &InitializeParticle(
    Color=&SelectLinearColorFromArray(
     Array Sampling Mode=::NewEnumerator0
     Direct Array Index=?
     Color Selection Array=DataInterface() {propertyValues {ColorData [
        <rgba 0.7 0.0583 0 1>
        <rgba 0 0.4667 1 1>
        <rgba 1 0.77 0 1>
        <rgba 1 1 1 1>
        <rgba 0.2479 0.2479 0.2479 1>
        <rgba 1 0.3333 0 1>
        <rgba 0 0.0667 1 1>
       ]}}
     Random Seed=0
    )
    Lifetime=&MaskFloatBySpawnGroup(Masked Float=0.3, Passthrough Float=&RandomRangeFloat(Minimum=0.15, Maximum=0.25))
    Position Offset=&AddVectorToPosition(
     Position=&AddVectorToPosition(Position=<0 0 0>, Vector=$User.SingleBoostPos)
     Vector=$User.DTCarOffset
    )
    Position Offset Coordinate Space=::Local
    Sprite Rotation Mode=::NewEnumerator2
    Sprite Size Max=<3 12>
    Sprite Size Min=<4 15>
    Sprite Size Mode=::NewEnumerator2
    UsePositionOffset=true
    Write Material Random=true
   )
   &PartitionParticles(Number Of Partitions=$Emitter.PortCount) {enabled=false}
   &SkeletalMeshLocation(
    Bone / Socket Index=$Particles.PartitionParticles.Partition
    Bone / Socket Sampling Mode=::NewEnumerator5
    MaskSpawnGroup=false
    Mesh Sampling Type=::NewEnumerator6
    Sampled Position Offset=&AddVectorToPosition(Position=<-20 0 0>, Vector=$User.DTCarOffset)
    Skeletal Mesh=$System.SkeletalMesh
   ) {enabled=false}
   &AddVelocity(
    Cone Angle=26
    Cone Axis=<-1 0 0.1>
    Module.UseConeFalloff=false
    Module.UseVelDistribution=false
    Override Seed=false
    Velocity Mode=::NewEnumerator2
    Velocity Speed=&MaskFloatBySpawnGroup(
     Masked Float=&RandomRangeFloat(Minimum=1500, Maximum=2000)
     Passthrough Float=&RandomRangeFloat(Minimum=1000, Maximum=2000)
    )
   )
   &ShapeLocation(
    Box / Plane Mode:ENiagara_BoxPlaneMode=::NewEnumerator1
    MaskSpawnGroup=false
    Override Seed=false
    Plane Size:Vector2f=<70 100>
    Rotation Mode=::NewEnumerator1
    Shape Primitive=::NewEnumerator2
    UseOwnerScale=true
   )
  ]
  particleUpdateScript [
   &ParticleState()
   &ScaleSpriteSize(
    Non-Uniform Curve Scale=&RandomRangeVector2D(Minimum=<1.35 1>, Maximum=<1.5 1.5>)
    Non-Uniform Curve Sprite Scale=DataInterface() {propertyValues {
      XCurve {keys [<at 0 0.875 RCIM_Cubic>, {arriveTangent=-2.6803, interpMode=RCIM_Cubic, leaveTangent=-2.6803, tangentMode=RCTM_User, time=0.7, value=0}]}
      YCurve {keys [<at 0 0.875 RCIM_Cubic>, {arriveTangent=-1.8085, interpMode=RCIM_Cubic, leaveTangent=-1.8085, tangentMode=RCTM_User, time=0.8, value=0}]}
     }}
    Scale Sprite Size Mode=::NewEnumerator3
   )
   &SpriteRotationRate(
    Delta Time:NiagaraFloat=$Particles.Age
    Rotation Rate=&MaskFloatBySpawnGroup(Masked Float=500, Passthrough Float=200)
   )
   &ScaleVelocity(
    Coordinate Space:ENiagaraCoordinateSpace=::Simulation
    Velocity Scale=&VectorFromCurve(
     VectorCurve=DataInterface() {propertyValues {
       XCurve {keys [<at 0 1>,<at 1 0.5>]}
       YCurve {keys [<at 0 1>,<at 1 1>]}
       ZCurve {keys [
         {arriveTangent=-4.9181, interpMode=RCIM_Cubic, leaveTangent=-4.9181, tangentMode=RCTM_User, time=0, value=0}
         {arriveTangent=17.7763, interpMode=RCIM_Cubic, leaveTangent=17.7763, tangentMode=RCTM_User, time=1, value=5}
        ]}
      }}
     Scale Curve=&RandomRangeVector(Minimum=<0.65 1 -2>, Maximum=<1.5 1 -2>)
    )
   )
   &InheritSourceMovement(
    Applied Position Scale=&Multiply_VectorByFloat(
     Vector=<1 1 1>
     Float=&FloatFromCurve(FloatCurve=DataInterface() {propertyValues {Curve {keys [
          {interpMode=RCIM_Cubic, tangentMode=RCTM_User, time=0, value=1}
          {arriveTangentWeight=0.3618, interpMode=RCIM_Cubic, tangentMode=RCTM_User, time=1, value=0.9}
         ]}}})
    )
    ApplyForce=false
    ApplyPosition=true
    ApplyPositionFalloff=false
   )
   &SolveForcesAndVelocity(Clamp Velocity=false, Limit Acceleration=false)
   &DynamicMaterialParameters(
    Index 0 Param 1=1
    Index 0 Param 2=&RandomRangeFloat(Minimum=1, Maximum=0)
    Index 0 Param 3=&RandomRangeFloat(Minimum=0, Maximum=1)
    Index 0 Param 4=&RandomRangeFloat(Minimum=0, Maximum=0.1)
    Module.Index0Param0Write=true
    Module.Index0Param1Write=true
    Module.Index0Param2Write=true
    Module.Index0Param3Write=true
   ) {enabled=false}
   &ScaleColor(Scale Alpha=0.8, Scale RGB=&VectorFromFloat(Value=0.4), ScaleA=true, ScaleRGB=true)
  ]
  rendererClasses [&NiagaraSpriteRendererProperties]
  renderers [
   {rendererClass=&NiagaraSpriteRendererProperties}
   {rendererClass=&NiagaraSpriteRendererProperties, rendererIndex=1}
  ]
 }
}
summary {
 systemName=NS_Boost_Kittybasic
 emitters [
  {emitterName=Sprite-Confetti, simTarget=GPUComputeSim, rendererClasses [&NiagaraSpriteRendererProperties]}
  {
   emitterName=NE_VC_Boost_Mesh_Box001
   rendererClasses [&NiagaraMeshRendererProperties, &NiagaraSpriteRendererProperties]
  }
  {
   emitterName=NE_VC_Boost_Mesh_Box
   rendererClasses [&NiagaraMeshRendererProperties, &NiagaraSpriteRendererProperties]
  }
 ]
 userVariables [
  User.bIsFrontendPreview {type=NiagaraBool}
  User.bIsPainted {type=NiagaraBool}
  User.ColorVariantID {type=NiagaraInt32}
  User.DTCarOffset {type=Vector3f, description="", defaultValue {struct=Vector3f, value=<0 0 0>}}
  User.PaintedColor {type=LinearColor, description="", defaultValue {struct=LinearColor, value=<rgba 1 1 1 1>}}
  User.PaintedScalar {type=NiagaraFloat, defaultValue=1, description=""}
  User.PortCount {type=NiagaraInt32, defaultValue=1}
  User.SingleBoostPos {type=Vector3f, description="", defaultValue {struct=Vector3f, value=<0 0 0>}}
 ]
}
